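# Physics asset for the MatchaSpare backpack cosmetic, defining capsule bodies and constraints for its body, two legs and three five-link button chains.
# The body links through mid to lower segments with narrow swing and locked twist.
# Center and right chains hang from the mid segment, while the left chain and both legs attach to the lower one.
# Button joints allow wider swing with small twist, tightest on the left chain.
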PhysicsAsset /BRCosmetics/Accessories/FORT_Backpacks/Backpack_M_MED_MatchaSpare/Meshes/M_MED_MatchaSpare_Pack_PhysicsButtons
bodies {
 dyn_body [{shapeType=Capsule, length=6.8171, radius=9.3371, shapeName=dyn_body_capsule, center=<3.9614 0.2513 0.097>, rotation=<rot -90 0 0>}]
 dyn_body_lower [{shapeType=Capsule, length=2, radius=2, shapeName=dyn_body_lower_capsule, center=<2 0 0>}]
 dyn_body_mid [{shapeType=Capsule, length=2, radius=2, shapeName=dyn_body_mid_capsule, center=<2 0 0>}]
 dyn_button_01_c [{shapeType=Capsule, length=1, radius=0.65, shapeName=dyn_button_01_c_capsule, center=<2 0 0>, rotation=<rot 90 0 0>}]
 dyn_button_01_l [{shapeType=Capsule, length=1, radius=0.65, shapeName=dyn_button_01_l_capsule, center=<2 0 0>, rotation=<rot 90 0 0>}]
 dyn_button_01_r [{shapeType=Capsule, length=1, radius=0.65, shapeName=dyn_button_01_r_capsule, center=<2 0 0>, rotation=<rot 90 0 0>}]
 dyn_button_02_c [{shapeType=Capsule, length=1, radius=0.65, shapeName=dyn_button_02_c_capsule, center=<2 0 0>, rotation=<rot 90 0 0>}]
 dyn_button_02_l [{shapeType=Capsule, length=1, radius=0.65, shapeName=dyn_button_02_l_capsule, center=<2 0 0>, rotation=<rot 90 0 0>}]
 dyn_button_02_r [{shapeType=Capsule, length=1, radius=0.65, shapeName=dyn_button_02_r_capsule, center=<2 0 0>, rotation=<rot 90 0 0>}]
 dyn_button_03_c [{shapeType=Capsule, length=1, radius=0.65, shapeName=dyn_button_03_c_capsule, center=<2 0 0>, rotation=<rot 90 0 0>}]
 dyn_button_03_l [{shapeType=Capsule, length=1, radius=0.65, shapeName=dyn_button_03_l_capsule, center=<2 0 0>, rotation=<rot 90 0 0>}]
 dyn_button_03_r [{shapeType=Capsule, length=1, radius=0.65, shapeName=dyn_button_03_r_capsule, center=<2 0 0>, rotation=<rot 90 0 0>}]
 dyn_button_04_c [{shapeType=Capsule, length=1, radius=1.15, shapeName=dyn_button_04_c_capsule, center=<2 0 0>, rotation=<rot 90 0 0>}]
 dyn_button_04_l [{shapeType=Capsule, length=1, radius=0.65, shapeName=dyn_button_04_l_capsule, center=<2 0 0>, rotation=<rot 90 0 0>}]
 dyn_button_04_r [{shapeType=Capsule, length=1, radius=1, shapeName=dyn_button_04_r_capsule, center=<2 0 0>, rotation=<rot 90 0 0>}]
 dyn_button_05_c [{
   shapeType=Capsule
   length=1.2951
   radius=1.9155
   shapeName=dyn_button_05_c_capsule
   center=<1.0311 0.0678 -0.5805>
   rotation=<rot 90 -94.0058 0>
  }]
 dyn_button_05_l [{shapeType=Capsule, length=0.4617, radius=1.9155, shapeName=dyn_button_05_l_capsule, center=<2 0 0>, rotation=<rot 90 0 0>}]
 dyn_button_05_r [{shapeType=Capsule, length=1, radius=1, shapeName=dyn_button_05_r_capsule, center=<2 0 0>, rotation=<rot 90 0 0>}]
 dyn_leg_01_l [{shapeType=Capsule, length=1, radius=1.2, shapeName=dyn_leg_01_l_capsule, center=<2 0 0>, rotation=<rot 90 0 0>}]
 dyn_leg_01_r [{shapeType=Capsule, length=1, radius=1.2, shapeName=dyn_leg_01_r_capsule, center=<2 0 0>, rotation=<rot 90 0 0>}]
}
constraints [
 {
  bone1Name=dyn_body_mid
  bone2Name=dyn_body
  swing1LimitDegrees=5
  swing1Motion=Locked
  swing2LimitDegrees=6
  twistLimitDegrees=45
  twistMotion=Locked
 }
 {bone1Name=dyn_button_01_r, bone2Name=dyn_body_mid, swing1LimitDegrees=40, swing2LimitDegrees=55, twistLimitDegrees=10}
 {bone1Name=dyn_button_01_c, bone2Name=dyn_body_mid, swing1LimitDegrees=40, swing2LimitDegrees=55, twistLimitDegrees=20}
 {bone1Name=dyn_body_lower, bone2Name=dyn_body_mid, swing1LimitDegrees=5, swing2LimitDegrees=6, twistLimitDegrees=45, twistMotion=Locked}
 {bone1Name=dyn_button_02_c, bone2Name=dyn_button_01_c, swing1LimitDegrees=40, swing2LimitDegrees=40, twistLimitDegrees=20}
 {bone1Name=dyn_button_03_c, bone2Name=dyn_button_02_c, swing1LimitDegrees=40, swing2LimitDegrees=40, twistLimitDegrees=20}
 {bone1Name=dyn_button_04_c, bone2Name=dyn_button_03_c, swing1LimitDegrees=40, swing2LimitDegrees=40, twistLimitDegrees=20}
 {bone1Name=dyn_button_05_c, bone2Name=dyn_button_04_c, swing1LimitDegrees=40, swing2LimitDegrees=40, twistLimitDegrees=20}
 {bone1Name=dyn_button_02_r, bone2Name=dyn_button_01_r, swing1LimitDegrees=40, swing2LimitDegrees=50, twistLimitDegrees=10}
 {bone1Name=dyn_button_03_r, bone2Name=dyn_button_02_r, swing1LimitDegrees=40, swing2LimitDegrees=40, twistLimitDegrees=10}
 {bone1Name=dyn_button_04_r, bone2Name=dyn_button_03_r, swing1LimitDegrees=40, swing2LimitDegrees=40, twistLimitDegrees=10}
 {bone1Name=dyn_button_05_r, bone2Name=dyn_button_04_r, swing1LimitDegrees=40, swing2LimitDegrees=40, twistLimitDegrees=10}
 {bone1Name=dyn_button_01_l, bone2Name=dyn_body_lower, swing1LimitDegrees=25, swing2LimitDegrees=55, twistLimitDegrees=5}
 {bone1Name=dyn_leg_01_l, bone2Name=dyn_body_lower, swing1LimitDegrees=15, swing2LimitDegrees=30, twistLimitDegrees=45, twistMotion=Locked}
 {bone1Name=dyn_leg_01_r, bone2Name=dyn_body_lower, swing1LimitDegrees=15, swing2LimitDegrees=30, twistLimitDegrees=45, twistMotion=Locked}
 {bone1Name=dyn_button_02_l, bone2Name=dyn_button_01_l, swing1LimitDegrees=25, swing2LimitDegrees=40, twistLimitDegrees=5}
 {bone1Name=dyn_button_03_l, bone2Name=dyn_button_02_l, swing1LimitDegrees=25, swing2LimitDegrees=40, twistLimitDegrees=5}
 {bone1Name=dyn_button_04_l, bone2Name=dyn_button_03_l, swing1LimitDegrees=25, swing2LimitDegrees=40, twistLimitDegrees=5}
 {bone1Name=dyn_button_05_l, bone2Name=dyn_button_04_l, swing1LimitDegrees=25, swing2LimitDegrees=40, twistLimitDegrees=5}
]
