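# Anim Blueprint for the MatchaSpare backpack, adding physics motion and reacting to the wearer's health.
# Chained RigidBody sims move body and buttons; low health blends to a body-only sim with offset buttons, before fixed arm, leg and button rotations.
# Health comes from WatchPawnHealth for local-team pawns, while the front end cycles a simulated drain.
# Health thresholds drive Hide Element parameters on dynamic and overlay materials.
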AnimBlueprint /BRCosmetics/Accessories/FORT_Backpacks/Backpack_M_MED_MatchaSpare/Meshes/M_MED_MatchaSpare_Pack_AnimBP
target_skeleton=&/M_MED_MatchaSpare_Pack_Skeleton
graphs [
 AnimGraph {nodes [
   n1 = RigidBody!(
    ComponentPose=Local To Component(LocalPose=Use cached pose 'StartCache'())
    Alpha
    OverrideWorldGravity:Vector="0,0,0"
    ExternalForce
    SimSpaceSettings:Sim Space Settings Structure=(WorldAlpha=1,VelocityScaleZ=1,DampingAlpha=1,MaxLinearVelocity=10000,MaxAngularVelocity=10000,MaxLinearAcceleration=10000,MaxAngularAcceleration=10000,ExternalLinearDragV=(X=0,Y=0,Z=0),ExternalLinearVelocity=(X=0,Y=0,Z=0),ExternalAngularVelocity=(X=0,Y=0,Z=0))
   ) -> (Pose) {
    bOverrideWorldGravity=true
    overridePhysicsAsset=&/M_MED_MatchaSpare_Pack_PhysicsNotButtons
    simulationSpace=BaseBoneSpace
    alphaBoolBlend {blendOption=Linear}
    baseBoneRef {boneName=root}
    componentAppliedLinearAccClamp=<3000 3000 3000>
    simSpaceSettings {worldAlpha=1}
   }
   n2 = RigidBody!(ComponentPose=@n1, Alpha, OverrideWorldGravity="0,0,0", ExternalForce) -> (Pose) {
    bOverrideWorldGravity=true
    overridePhysicsAsset=&/M_MED_MatchaSpare_Pack_PhysicsButtons
    simulationSpace=BaseBoneSpace
    alphaBoolBlend {blendOption=Linear}
    baseBoneRef {boneName=root}
    componentAppliedLinearAccClamp=<3000 3000 3000>
    simSpaceSettings {
     dampingAlpha=2
     maxAngularAcceleration=700
     maxAngularVelocity=700
     maxLinearAcceleration=1000
     maxLinearVelocity=700
     velocityScaleZ=0.5
     worldAlpha=0.5
     externalLinearDragV=<0.5 3 5>
    }
   }
   n3 = "Constraint - base_button_r (PP)"!(ComponentPose=@n2, Alpha=1, ConstraintWeights_0=0.6013, ConstraintWeights_1:float=0.3987) -> (Pose) {
    alphaBoolBlend {blendOption=Linear}
    boneToModify {boneName=base_button_r}
    constraintSetup [{targetBone {boneName=dyn_body_mid}}, {targetBone {boneName=dyn_body_lower}}]
    constraintWeights [1,1]
   }
   n4 = Component To Local(ComponentPose=@n3)
   n5 = Use cached pose 'SurfaceButton'()
   n6 = Get LowHealth() -> (LowHealth:bool)
   n7 = "Transform (Modify) Bone - Bone: base_arm_r"(
    ComponentPose=Local To Component(LocalPose=Blend Poses by bool(BlendPose_0=@n4, BlendPose_1=@n5, bActiveValue=@n6) {blendType=Linear})
    Alpha=1
    Translation="0,0,0"
    Rotation="0,40,0"
    Scale="1,1,1"
   ) {
    rotationMode=BMM_Additive
    rotationSpace=BCS_BoneSpace
    alphaBoolBlend {blendOption=Linear}
    alphaScaleBiasClamp {bInterpResult=true, interpSpeedDecreasing=5, interpSpeedIncreasing=5}
    boneToModify {boneName=base_arm_r}
    rotation {yaw=40}
   }
   n8 = "Transform (Modify) Bone - Bone: base_arm_l"(ComponentPose=@n7, Alpha=1, Translation="0,0,0", Rotation="0,-40,0", Scale="1,1,1") {
    rotationMode=BMM_Additive
    rotationSpace=BCS_BoneSpace
    alphaBoolBlend {blendOption=Linear}
    alphaScaleBiasClamp {bInterpResult=true, interpSpeedDecreasing=5, interpSpeedIncreasing=5}
    boneToModify {boneName=base_arm_l}
    rotation {yaw=-40}
   }
   n9 = "Transform (Modify) Bone - Bone: base_leg_r"(ComponentPose=@n8, Alpha=1, Translation="0,0,0", Rotation="45,0,0", Scale="1,1,1") {
    rotationMode=BMM_Additive
    rotationSpace=BCS_BoneSpace
    alphaBoolBlend {blendOption=Linear}
    alphaScaleBiasClamp {bInterpResult=true, interpSpeedDecreasing=5, interpSpeedIncreasing=5}
    boneToModify {boneName=base_leg_r}
    rotation {pitch=45}
   }
   n10 = "Transform (Modify) Bone - Bone: base_leg_l"(ComponentPose=@n9, Alpha=1, Translation="0,0,0", Rotation="39.159,-22.692,-19.052", Scale="1,1,1") {
    rotationMode=BMM_Additive
    rotationSpace=BCS_BoneSpace
    alphaBoolBlend {blendOption=Linear}
    alphaScaleBiasClamp {bInterpResult=true, interpSpeedDecreasing=5, interpSpeedIncreasing=5}
    boneToModify {boneName=base_leg_l}
    rotation=<rot 39.159 -22.692 -19.052>
   }
   n11 = "Transform (Modify) Bone - Bone: base_button_r"(ComponentPose=@n10, Alpha=1, Translation="0,0,0", Rotation="4.225,4.241,23.892", Scale="1,1,1") {
    rotationMode=BMM_Additive
    rotationSpace=BCS_BoneSpace
    alphaBoolBlend {blendOption=Linear}
    alphaScaleBiasClamp {bInterpResult=true, interpSpeedDecreasing=5, interpSpeedIncreasing=5}
    boneToModify {boneName=base_button_r}
    rotation=<rot 4.225 4.241 23.892>
   }
   n12 = "Transform (Modify) Bone - Bone: base_button_l"(ComponentPose=@n11, Alpha=1, Translation="0,0,0", Rotation="-20.053,-5.829,-33.233", Scale="1,1,1") {
    rotationMode=BMM_Additive
    rotationSpace=BCS_BoneSpace
    alphaBoolBlend {blendOption=Linear}
    alphaScaleBiasClamp {bInterpResult=true, interpSpeedDecreasing=5, interpSpeedIncreasing=5}
    boneToModify {boneName=base_button_l}
    rotation=<rot -20.053 -5.829 -33.233>
   }
   Output Pose(Result=Component To Local(ComponentPose=@n12)) {layerGroup=None}
   Save cached pose 'StartCache'(Pose=Input Pose()) {cacheName=StartCache}
   n13 = RigidBody!(
    ComponentPose=Local To Component(LocalPose=Use cached pose 'StartCache'())
    Alpha
    OverrideWorldGravity="0,0,0"
    ExternalForce
    SimSpaceSettings:Sim Space Settings Structure=(WorldAlpha=1,VelocityScaleZ=1,DampingAlpha=1,MaxLinearVelocity=10000,MaxAngularVelocity=10000,MaxLinearAcceleration=10000,MaxAngularAcceleration=10000,ExternalLinearDragV=(X=0,Y=0,Z=0),ExternalLinearVelocity=(X=0,Y=0,Z=0),ExternalAngularVelocity=(X=0,Y=0,Z=0))
   ) -> (Pose) {
    bOverrideWorldGravity=true
    overridePhysicsAsset=&M_MED_MatchaSpare_Pack_PhysicsNotButtons
    simulationSpace=BaseBoneSpace
    alphaBoolBlend {blendOption=Linear}
    baseBoneRef {boneName=root}
    componentAppliedLinearAccClamp=<3000 3000 3000>
    simSpaceSettings {worldAlpha=1}
   }
   n14 = "Constraint - base_button_r (PP)"(ComponentPose=@n13, Alpha=1, ConstraintWeights_0=0.6013, ConstraintWeights_1=0.3987) {
    alphaBoolBlend {blendOption=Linear}
    boneToModify {boneName=base_button_r}
    constraintSetup [{targetBone {boneName=dyn_body_mid}}, {targetBone {boneName=dyn_body_lower}}]
    constraintWeights [1,1]
   }
   n15 = "Transform (Modify) Bone - Bone: base_button_l"(ComponentPose=@n14, Alpha=1, Translation="-9.145,-0.105,-0.044", Rotation="0,0,0", Scale="1,1,1") {
    rotationMode=BMM_Additive
    rotationSpace=BCS_BoneSpace
    translationMode=BMM_Additive
    translationSpace=BCS_BoneSpace
    alphaBoolBlend {blendOption=Linear}
    boneToModify {boneName=base_button_l}
    translation=<-9.145 -0.105 -0.044>
   }
   n16 = "Transform (Modify) Bone - Bone: base_button_r"(ComponentPose=@n15, Alpha=1, Translation="-9.118,-0.101,-0.025", Rotation="0,0,0", Scale="1,1,1") {
    rotationMode=BMM_Additive
    rotationSpace=BCS_BoneSpace
    translationMode=BMM_Additive
    translationSpace=BCS_BoneSpace
    alphaBoolBlend {blendOption=Linear}
    boneToModify {boneName=base_button_r}
    translation=<-9.118 -0.101 -0.025>
   }
   n17 = "Transform (Modify) Bone - Bone: base_button_c"(ComponentPose=@n16, Alpha=1, Translation="-7.427,-0.005,-0.218", Rotation="3.696,0.032,-0.001", Scale="1,1,1") {
    rotationMode=BMM_Additive
    rotationSpace=BCS_BoneSpace
    translationMode=BMM_Additive
    translationSpace=BCS_BoneSpace
    alphaBoolBlend {blendOption=Linear}
    boneToModify {boneName=base_button_c}
    rotation=<rot 3.696 0.032 -0.001>
    translation=<-7.4271 -0.005 -0.218>
   }
   Save cached pose 'SurfaceButton'(Pose=Component To Local(ComponentPose=@n17)) {cacheName=SurfaceButton}
  ]}
 EventGraph {nodes [
   n1 = Event BlueprintUpdateAnimation()
   > Set deltaTime(deltaTime:double=@n1.DeltaTimeX) -> (Output_Get:double)
   > n2 = Sequence!(execute) -> (then_0, then_1, then_2:Exec, then_3:Exec)
   > n3 = "Gravity Override State Function (Parameter Asset) (Nativized)"(
    CharacterPartAnimInstance=Self-Reference()
    GameTimeInSeconds=GetGameTimeInSeconds()
    DebugJointName=spine_05
    Parameters=&/M_MED_MatchaSpare_Pack_Parameters
   )
   > Set Gravity Override(Gravity Override=@n3)
   > Set RBAN Settings Out(RBAN Settings Out=@n3.OutRigidBodyAnimNodeParameters)
   Branch(execute=@n2.then_2, Condition=Get bIsInFrontEnd())
   > Set cumulative time(cumulative time:double=float + float(A=Get deltaTime() -> (deltaTime:double), B=Get cumulative time() -> (cumulative time:double))) -> (Output_Get:double)
   > Branch(Condition="float >= float"(A=Get cumulative time(), B=Get interval() -> (interval:double)))
   > Set healthValue(healthValue:double=float - float(A=Get healthValue() -> (healthValue:double), B=15)) -> (Output_Get:double)
   > Set cumulative time(cumulative time=0)
   > Branch(Condition="float <= float"(A=Get healthValue(), B=40) -> (ReturnValue=~))
   > Set LowHealth(LowHealth:bool=true) -> (Output_Get:bool)
   > Branch(Condition="float <= float"(A=Get healthValue()) -> (ReturnValue=~))
   > Set healthValue(healthValue=100)
   > Set LowHealth(LowHealth=false)
   n4 = Sequence(execute=@n2.then_3)
   n5 = Get healthValue()
   > n6 = Branch(Condition="float < float"(A=@n5, B=40))
   n7 = Get DynamicMaterial() -> (DynamicMaterial:Material Instance Dynamic Object Reference)
   > SetScalarParameterValue(self=@n7, ParameterName=Hide Element 01, Value=0)
   n8 = GetPartSkeletalMeshComponent()
   > SetScalarParameterValueOnOverlayMaterial!(execute, MeshComponent:Mesh Component Object Reference=@n8, ParameterName:Name=Hide Element 02, ParameterValue:float=0) -> (then)
   SetScalarParameterValue(execute=@n6.else, self=@n7, ParameterName=Hide Element 01, Value=0.1)
   > SetScalarParameterValueOnOverlayMaterial(MeshComponent=@n8, ParameterName=Hide Element 02, ParameterValue=0.3)
   n9 = Branch(execute=@n4.then_1, Condition="float < float"(A=Get healthValue(), B=100))
   n10 = Get DynamicMaterial()
   > SetScalarParameterValue(self=@n10, ParameterName=Hide Element 02, Value=0)
   > SetScalarParameterValueOnOverlayMaterial(MeshComponent=@n8, ParameterName=Hide Element 02, ParameterValue=0)
   SetScalarParameterValue(execute=@n9.else, self=@n10, ParameterName=Hide Element 02, Value=0.2)
   > SetScalarParameterValueOnOverlayMaterial(MeshComponent=@n8, ParameterName=Hide Element 02, ParameterValue=0.3)
   n11 = Get healthValue()
   n12 = Branch(execute=@n4.then_2, Condition="float < float"(A=@n11, B=80))
   n13 = Get DynamicMaterial()
   > SetScalarParameterValue(self=@n13, ParameterName=Hide Element 03, Value=0)
   > SetScalarParameterValueOnOverlayMaterial(MeshComponent=@n8, ParameterName=Hide Element 03, ParameterValue=0)
   SetScalarParameterValue(execute=@n12.else, self=@n13, ParameterName=Hide Element 03, Value=0.3)
   > SetScalarParameterValueOnOverlayMaterial(MeshComponent=@n8, ParameterName=Hide Element 03, ParameterValue=0.3)
   n14 = Get healthValue()
   n15 = Branch(execute=@n4.then_3, Condition="float < float"(A=@n14, B=60))
   n16 = Get DynamicMaterial()
   > SetScalarParameterValue(self=@n16, ParameterName=Hide Element 04, Value=0)
   > SetScalarParameterValueOnOverlayMaterial(MeshComponent=@n8, ParameterName=Hide Element 04, ParameterValue=0)
   SetScalarParameterValue(execute=@n15.else, self=@n16, ParameterName=Hide Element 04, Value=0.4)
   > SetScalarParameterValueOnOverlayMaterial(MeshComponent=@n8, ParameterName=Hide Element 04, ParameterValue=0.3)
   n17 = Get Dynamic Material Instance(execute=@n2.then_1, SkeletalMeshComponent=GetPartSkeletalMeshComponent())
   > Set DynamicMaterial(DynamicMaterial:Material Instance Dynamic Object Reference=@n17) -> (Output_Get:Material Instance Dynamic Object Reference)
   n18 = Get bIsCrouching()
   n19 = Get bIsMoving2D()
   > Set Alpha_moving2dCrouch(Alpha_moving2dCrouch:double="To Float (Boolean)"(InBool=AND Boolean(A=@n19, B=@n18))) -> (Output_Get:double)
   Event BlueprintInitializeAnimation()
   > Set Skel_Mesh(Skel_Mesh=GetOwnerSkeletalMeshComponent())
   > n20 = Cast To FortPlayerPawn(Object=TryGetFortPawn())
   > n21 = Set Pawn(Pawn:Fort Player Pawn Object Reference=@n20) -> (Output_Get:Fort Player Pawn Object Reference)
   > Branch(Condition=OnLocalPlayersTeam!(TargetActor:Actor Object Reference=@n21) -> (Return Value:bool))
   > n22 = WatchPawnHealth!(execute:Exec, PlayerPawn:Fort Player Pawn Object Reference=@n21) -> (then:Exec, OnHitPointsChanged:Exec, Health:float=~, MaxHealth:float=~, Shield:float=~, MaxShield:float=~) {is=title K2Node_AsyncAction}
   Set healthValue(execute=@n22.OnHitPointsChanged, healthValue=@n22)
   > n23 = Branch(Condition="float < float"(A=Get healthValue(), B=40))
   > Set LowHealth(LowHealth=true)
   Set LowHealth(execute=@n23.else, LowHealth=false)
   "float <= float"(A=@n5, B=40) -> (ReturnValue=~)
   "float <= float"(A=@n11, B=80) -> (ReturnValue=~)
   "float <= float"(A=@n14, B=60) -> (ReturnValue=~)
   Get Damage1() -> (Damage1:bool)
   Get Damage2() -> (Damage2:bool)
   Get Damage3() -> (Damage3:bool)
  ]}
]
variables [
 Alpha_moving2dCrouch {type=number}
 cumulative time {type=number}
 damage received {type=number}
 Damage1 {type=boolean}
 Damage2 {type=boolean}
 Damage3 {type=boolean}
 deltaTime {type=number}
 DynamicMaterial {type=MaterialInstanceDynamic}
 Gravity Override {type=Vector}
 healthValue {type=number, default=100}
 interval {type=number, default=2}
 LowHealth {type=boolean}
 Pawn {type=FortPlayerPawn}
 RBAN Settings Out {type=FortRigidBodyAnimNodeParameters}
 Skel_Mesh {type=SkeletalMeshComponent}
]
