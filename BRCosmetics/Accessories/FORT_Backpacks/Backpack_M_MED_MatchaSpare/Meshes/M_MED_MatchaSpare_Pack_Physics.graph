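# Physics asset for the MatchaSpare backpack cosmetic mesh, defining collision bodies and joints for a dangling character with antlers, arms and legs.
# Small capsules cover each chained bone, while a larger capsule on dyn_body acts as the root.
# Head with antlers, mid body with four-segment arms, and lower body with three-segment legs all branch from that root.
# Every joint locks twist and limits swing, tightest at the head.
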PhysicsAsset /BRCosmetics/Accessories/FORT_Backpacks/Backpack_M_MED_MatchaSpare/Meshes/M_MED_MatchaSpare_Pack_Physics
bodies {
 antlers_01_l [{shapeType=Capsule, length=1, radius=1, shapeName=antlers_01_l_capsule, center=<2 0 0>, rotation=<rot 90 0 0>}]
 antlers_01_r [{shapeType=Capsule, length=1, radius=1, shapeName=antlers_01_r_capsule, center=<2 0 0>, rotation=<rot 90 0 0>}]
 antlers_02_l [{shapeType=Capsule, length=1, radius=1, shapeName=antlers_02_l_capsule, center=<2 0 0>, rotation=<rot 90 0 0>}]
 antlers_02_r [{shapeType=Capsule, length=1, radius=1, shapeName=antlers_02_r_capsule, center=<2 0 0>, rotation=<rot 90 0 0>}]
 dyn_arm_01_l [{shapeType=Capsule, length=1, radius=1, shapeName=dyn_arm_01_l_capsule, center=<2 0 0>, rotation=<rot 90 0 0>}]
 dyn_arm_01_r [{shapeType=Capsule, length=1, radius=1, shapeName=dyn_arm_01_r_capsule, center=<2 0 0>, rotation=<rot 90 0 0>}]
 dyn_arm_02_l [{shapeType=Capsule, length=1, radius=1, shapeName=dyn_arm_02_l_capsule, center=<2 0 0>, rotation=<rot 90 0 0>}]
 dyn_arm_02_r [{shapeType=Capsule, length=1, radius=1, shapeName=dyn_arm_02_r_capsule, center=<2 0 0>, rotation=<rot 90 0 0>}]
 dyn_arm_03_l [{shapeType=Capsule, length=1, radius=1, shapeName=dyn_arm_03_l_capsule, center=<2 0 0>, rotation=<rot 90 0 0>}]
 dyn_arm_03_r [{shapeType=Capsule, length=1, radius=1, shapeName=dyn_arm_03_r_capsule, center=<2 0 0>, rotation=<rot 90 0 0>}]
 dyn_arm_04_l [{shapeType=Capsule, length=1, radius=1, shapeName=dyn_arm_04_l_capsule, center=<2 0 0>, rotation=<rot 90 0 0>}]
 dyn_arm_04_r [{shapeType=Capsule, length=1, radius=1, shapeName=dyn_arm_04_r_capsule, center=<2 0 0>, rotation=<rot 90 0 0>}]
 dyn_body [{
   shapeType=Capsule
   length=6.8171
   radius=9
   shapeName=dyn_body_capsule
   center=<-2.4369 0.2274 0.1449>
   rotation=<rot 87.2882 25.3926 17.9922>
  }]
 dyn_body_lower [{shapeType=Capsule, length=3.0563, radius=2.6374, shapeName=dyn_body_lower_capsule}]
 dyn_body_mid [{shapeType=Capsule, length=1, radius=1, shapeName=dyn_body_mid_capsule, rotation=<rot 90 0 0>}]
 dyn_head [{shapeType=Capsule, length=2, radius=2, shapeName=dyn_head_capsule, center=<2 0 0>}]
 dyn_leg_01_l [{shapeType=Capsule, length=1, radius=1.2, shapeName=dyn_leg_01_l_capsule, center=<2 0 0>, rotation=<rot 90 0 0>}]
 dyn_leg_01_r [{shapeType=Capsule, length=1, radius=1.2, shapeName=dyn_leg_01_r_capsule, center=<2 0 0>, rotation=<rot 90 0 0>}]
 dyn_leg_02_l [{shapeType=Capsule, length=1, radius=1.2, shapeName=dyn_leg_02_l_capsule, center=<2 0 0>, rotation=<rot 90 0 0>}]
 dyn_leg_02_r [{shapeType=Capsule, length=1, radius=1.2, shapeName=dyn_leg_02_r_capsule, center=<2 0 0>, rotation=<rot 90 0 0>}]
 dyn_leg_03_l [{shapeType=Capsule, length=1, radius=1.2, shapeName=dyn_leg_03_l_capsule, center=<2 0 0>, rotation=<rot 90 0 0>}]
 dyn_leg_03_r [{shapeType=Capsule, length=1, radius=1.2, shapeName=dyn_leg_03_r_capsule, center=<2 0 0>, rotation=<rot 90 0 0>}]
}
constraints [
 {bone1Name=dyn_arm_04_r, bone2Name=dyn_arm_03_r, swing1LimitDegrees=15, swing2LimitDegrees=30, twistLimitDegrees=45, twistMotion=Locked}
 {bone1Name=dyn_arm_03_r, bone2Name=dyn_arm_02_r, swing1LimitDegrees=15, swing2LimitDegrees=30, twistLimitDegrees=45, twistMotion=Locked}
 {bone1Name=dyn_arm_02_r, bone2Name=dyn_arm_01_r, swing1LimitDegrees=15, swing2LimitDegrees=30, twistLimitDegrees=45, twistMotion=Locked}
 {bone1Name=dyn_arm_04_l, bone2Name=dyn_arm_03_l, swing1LimitDegrees=15, swing2LimitDegrees=30, twistLimitDegrees=20, twistMotion=Locked}
 {bone1Name=dyn_arm_03_l, bone2Name=dyn_arm_02_l, swing1LimitDegrees=15, swing2LimitDegrees=30, twistLimitDegrees=20, twistMotion=Locked}
 {bone1Name=dyn_arm_02_l, bone2Name=dyn_arm_01_l, swing1LimitDegrees=15, swing2LimitDegrees=30, twistLimitDegrees=20, twistMotion=Locked}
 {bone1Name=dyn_arm_01_l, bone2Name=dyn_body_mid, swing1LimitDegrees=15, swing2LimitDegrees=30, twistLimitDegrees=20, twistMotion=Locked}
 {bone1Name=dyn_arm_01_r, bone2Name=dyn_body_mid, swing1LimitDegrees=15, swing2LimitDegrees=30, twistLimitDegrees=45, twistMotion=Locked}
 {bone1Name=dyn_body_lower, bone2Name=dyn_body, swing1LimitDegrees=5, swing2LimitDegrees=10, twistLimitDegrees=45, twistMotion=Locked}
 {bone1Name=dyn_leg_01_l, bone2Name=dyn_body_lower, swing1LimitDegrees=15, swing2LimitDegrees=30, twistLimitDegrees=45, twistMotion=Locked}
 {bone1Name=dyn_leg_02_l, bone2Name=dyn_leg_01_l, swing1LimitDegrees=15, swing2LimitDegrees=30, twistLimitDegrees=45, twistMotion=Locked}
 {bone1Name=dyn_leg_03_l, bone2Name=dyn_leg_02_l, swing1LimitDegrees=15, swing2LimitDegrees=30, twistLimitDegrees=45, twistMotion=Locked}
 {bone1Name=dyn_leg_01_r, bone2Name=dyn_body_lower, swing1LimitDegrees=15, swing2LimitDegrees=30, twistLimitDegrees=45, twistMotion=Locked}
 {bone1Name=dyn_leg_02_r, bone2Name=dyn_leg_01_r, swing1LimitDegrees=15, swing2LimitDegrees=30, twistLimitDegrees=45, twistMotion=Locked}
 {bone1Name=dyn_leg_03_r, bone2Name=dyn_leg_02_r, swing1LimitDegrees=15, swing2LimitDegrees=30, twistLimitDegrees=45, twistMotion=Locked}
 {bone1Name=antlers_02_l, bone2Name=antlers_01_l, swing1LimitDegrees=5, swing2LimitDegrees=8, twistLimitDegrees=45, twistMotion=Locked}
 {bone1Name=antlers_01_l, bone2Name=dyn_head, swing1LimitDegrees=5, swing2LimitDegrees=8, twistLimitDegrees=45, twistMotion=Locked}
 {bone1Name=antlers_02_r, bone2Name=antlers_01_r, swing1LimitDegrees=5, swing2LimitDegrees=8, twistLimitDegrees=45, twistMotion=Locked}
 {bone1Name=antlers_01_r, bone2Name=dyn_head, swing1LimitDegrees=5, swing2LimitDegrees=8, twistLimitDegrees=45, twistMotion=Locked}
 {bone1Name=dyn_head, bone2Name=dyn_body, swing1LimitDegrees=2.5, swing2LimitDegrees=2.5, twistLimitDegrees=45, twistMotion=Locked}
 {bone1Name=dyn_body_mid, bone2Name=dyn_body, swing1LimitDegrees=2.5, swing2LimitDegrees=12, twistLimitDegrees=45, twistMotion=Locked}
]
